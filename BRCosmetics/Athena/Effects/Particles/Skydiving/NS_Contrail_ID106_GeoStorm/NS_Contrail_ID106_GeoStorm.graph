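# Niagara skydiving contrail cosmetic "GeoStorm" (ID106), trailing glitchy ribbons, shapes and sparkles in blue, cyan, purple and green from the player.
# Ribbon, mesh, sprite and light emitters follow bones sampled from User.SkeletalMesh, using spawn groups and random color-curve picks.
# Diving attributes stretch bits and meshes, alter ribbon material parameters and drive the spawn rate of GPU speed lines.
# Front-end modifiers adjust some ribbon velocities and mesh scales.
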NiagaraSystem /BRCosmetics/Athena/Effects/Particles/Skydiving/NS_Contrail_ID106_GeoStorm/NS_Contrail_ID106_GeoStorm
dependencies {
 usedDataInterfaces [
  &NiagaraDataInterfaceColorCurve
  &NiagaraDataInterfaceCurve
  &NiagaraDataInterfaceVectorCurve
  &NiagaraDataInterfaceArrayFloat3
  &NiagaraDataInterfaceVector2DCurve
 ]
 usedDynamicInputs []
 usedModules [&SystemState]
}
emitters {
 bits {
  emitterSpawnScript [&FN_PreserveSkydiveUserVariables() {module=PreserveUserVariables}]
  emitterUpdateScript [
   &EmitterState()
   /Game/Effects/Niagara/Modules/Spawn/SpawnRateFromMultipleSpawnGroups_4(Global Spawnrate=5)
   /Game/Effects/Niagara/Modules/Skydiving/FN_DivingAttributes/FN_EmitterDivingAttributes_v1(SkeletalMesh:NiagaraDataInterfaceSkeletalMesh=$User.SkeletalMesh)
  ]
  particleSpawnScript [
   &FN_SetSpawnGroup_v2(Number of Spawn Groups=4)
   /Niagara/Modules/Spawn/Initialization/InitializeParticle(
    Lifetime=&UniformRangedFloat(Minimum=0.3, Maximum=0.5)
    Mesh Scale=&Multiply_Vector(
     A=<1 1 1>
     B=&VectorFromFloat(Value=&UniformRangedFloat(Minimum=0.2, Maximum=1))
    )
    Module.Write Color=false
    Module.Write Lifetime=true
    Module.Write Mass=true
    Module.Write Material Random=true
    Module.Write Position=false
    Module.Write Scale=false
    Module.Write SpriteRotation=true
    Module.Write SpriteSize=true
    Position=&SimulationPosition()
    Sprite Rotation=&UniformAOrBFloat(A=0, B=360)
    Sprite Size=&Multiply_Vector2(
     A=&RandomRangeVector2D(Minimum=<2 2>, Maximum=<5 5>)
     B=&Vector2DFromFloat(Value=2)
    )
   )
   &SkeletalMeshLocation(
    Bone / Socket Index=$Particles.SpawnGroup
    Bone Sampling Mode=::NewEnumerator4
    Inherit Velocity=false
    MaskSpawnGroup=false
    Skeletal Mesh=$User.SkeletalMesh
   )
   &AddVelocity(
    Coordinate Space=::World
    Scale Added Velocity=&Lerp_Vector(A=<1 1 1>, B=<1 1 3>, Alpha=$Emitter.Dive)
    Velocity=&UniformRangedVector(Minimum=<0 0 250>, Maximum=<1 1 400>)
   )
   /Niagara/Modules/Spawn/Location/SphereLocation(MaskSpawnGroup=false, Module.Override Randomness=false, Sphere Radius=12.5, Surface Only=false)
   /SetVariables(Particles.ScaleUpdate:Vector2f=&Multiply_Vector2(
     A=&RandomRangeVector2D(Minimum=<1 1>, Maximum=<5 5>)
     B=&Vector2DFromFloat(Value=5)
    )) {isSetParameters=true}
  ]
  particleUpdateScript [
   &ParticleState()
   &Drag(Drag=0.5, Use Linear Drag=true, Use Rotational Drag=false)
   /Niagara/Modules/Update/Forces/CurlNoiseForce(
    Noise Frequency:NiagaraFloat=&UniformRangedFloat(Minimum=0.5, Maximum=1)
    Noise Strength=&UniformRangedFloat(Minimum=50, Maximum=150)
    Pan Noise=false
    Pan Noise Field=<1 1 0>
   )
   &SolveForcesAndVelocity(Clamp Velocity=false, Limit Acceleration=false)
   /SetVariables(Particles.Color=&MakeLinearColorFromVectorAndFloat(
     "Vector (RGB)"=&MakeVectorFromLinearColorRGB(Color=&ColorFromCurve(
       ColorCurve=DataInterface() {propertyValues {
         AlphaCurve {keys [<at 0 1 RCIM_Cubic>,<at 1 1 RCIM_Cubic>]}
         BlueCurve {keys [<at 0 0>, <at 0.225 0.651>, {arriveTangent=0.4503, interpMode=RCIM_Cubic, leaveTangent=0.4503, time=0.6172, value=1}, <at 1 1 RCIM_Cubic>]}
         GreenCurve {keys [
           <at 0 0.599>
           <at 0.225 0.1163>
           {arriveTangent=-0.0426, interpMode=RCIM_Cubic, leaveTangent=-0.0426, time=0.6172, value=0.7687}
           <at 1 0.0833 RCIM_Cubic>
          ]}
         RedCurve {keys [
           <at 0 0.2938>
           <at 0.225 0>
           {arriveTangent=0.6861, interpMode=RCIM_Cubic, leaveTangent=0.6861, time=0.6172, value=0.0525}
           <at 1 0.5317 RCIM_Cubic>
          ]}
        }}
       CurveIndex=&RandomRangeFloat(Minimum=0, Maximum=1)
      ))
     "Float (Alpha)"=&FloatFromCurve(FloatCurve=DataInterface() {propertyValues {Curve {keys [<at 0 0>,<at 0.1936 0.8315>,<at 0.4019 0.1015>,<at 0.5589 0.6516>,<at 0.7294 0.0749>,<at 0.8673 0.41>,<at 1.0011 0.002>]}}})
    )) {isSetParameters=true}
   &ScaleSpriteSize(Scale Factor=&Vector2DFromCurve(
     Vector2Curve=DataInterface() {propertyValues {
       XCurve {keys [<at 0 1 RCIM_Constant>,<at 0.1 1.7 RCIM_Constant>,<at 0.3 1.1 RCIM_Constant>,<at 0.5 1.8 RCIM_Constant>,<at 0.7 1.1 RCIM_Constant>,<at 0.8 1.9 RCIM_Constant>,<at 0.9 1.4>]}
       YCurve {keys [<at 0 1 RCIM_Constant>,<at 0.2 2 RCIM_Constant>,<at 0.4 1.5 RCIM_Constant>,<at 0.5 1.5 RCIM_Constant>,<at 0.6 1.8 RCIM_Constant>,<at 0.7 1.3 RCIM_Constant>,<at 1 1>]}
      }}
     ScaleCurve=&MakeVector2D(
      X=&Lerp_Float(A=0.6, B=0.4, Alpha=$Emitter.Dive)
      Y=&Lerp_Float(A=0.6, B=4, Alpha=$Emitter.Dive)
     )
    )) {module=ScaleSpriteSize001}
  ]
  rendererClasses [&NiagaraSpriteRendererProperties]
  renderers [{rendererClass=&NiagaraSpriteRendererProperties}]
 }
 decal {
  emitterSpawnScript [&FN_PreserveSkydiveUserVariables() {module=PreserveUserVariables}]
  emitterUpdateScript [
   &EmitterState()
   &SpawnRateFromMultipleSpawnGroups_4(Global Spawnrate=15)
   &FN_EmitterDivingAttributes_v1(SkeletalMesh:NiagaraDataInterfaceSkeletalMesh=$User.SkeletalMesh)
  ]
  particleSpawnScript [
   &FN_SetSpawnGroup_v2(Number of Spawn Groups=4)
   &InitializeParticle(
    Color=&ColorFromCurve(
     ColorCurve=DataInterface() {propertyValues {
       AlphaCurve {keys [<at 0 1 RCIM_Cubic>,<at 0.3655 1>,<at 1 1 RCIM_Cubic>]}
       BlueCurve {keys [
         <at -0.875 1>
         <at 0 0>
         {arriveTangent=1.5422, interpMode=RCIM_Cubic, leaveTangent=1.5422, time=0.2839, value=0.651}
         <at 0.6484 1>
         <at 0.9321 1 RCIM_Cubic>
        ]}
       GreenCurve {keys [
         <at -0.875 1>
         <at 0 0.599>
         {arriveTangent=0.2618, interpMode=RCIM_Cubic, leaveTangent=0.2618, time=0.2839, value=0.1163}
         <at 0.6484 0.7687>
         <at 0.9321 0.0833 RCIM_Cubic>
        ]}
       RedCurve {keys [
         <at -0.875 1>
         <at 0 0.2938>
         {arriveTangent=-0.3721, interpMode=RCIM_Cubic, leaveTangent=-0.3721, time=0.2839, value=0}
         <at 0.6484 0.0525>
         <at 0.9321 0.5317 RCIM_Cubic>
        ]}
      }}
     CurveIndex=&RandomRangeFloat(Minimum=0, Maximum=1)
    )
    Lifetime=0.6
    Module.Write Color=true
    Module.Write Lifetime=true
    Module.Write Mass=true
    Module.Write Material Random=true
    Module.Write Position=false
    Module.Write Scale=false
    Module.Write SpriteRotation=true
    Module.Write SpriteSize=true
    Position=&SimulationPosition()
    Sprite Rotation=&RandomRangeFloat(Minimum=-10, Maximum=10)
    Sprite Size=&Multiply_Vector2(
     A=<60 25>
     B=&Vector2DFromFloat(Value=&RandomRangeFloat(Minimum=3, Maximum=5))
    )
   )
   &SkeletalMeshLocation(Inherit Velocity=false, MaskSpawnGroup=false, Module.Override Randomness=false, Skeletal Mesh=$User.SkeletalMesh) {module=SkeletalMeshLocation001}
   &AddVelocity(Velocity=<0 -300 0>)
  ]
  particleUpdateScript [
   &SolveForcesAndVelocity(Clamp Velocity=false, Limit Acceleration=false)
   &ParticleState()
   &SampleSkeletalMeshSkeleton(Mesh:NiagaraDataInterfaceSkeletalMesh=$User.SkeletalMesh, Specific Bone Index:NiagaraInt32=$Particles.SpawnGroup)
   &SkeletalMeshLocation(
    Inherit Velocity=false
    MaskSpawnGroup=false
    Module.Override Randomness=false
    Sampled Position Offset=&RandomRangeVector(Minimum=<-2 5 0>, Maximum=<2 7 1>)
    Skeletal Mesh=$User.SkeletalMesh
   ) {enabled=false}
   &ScaleColor(
    Scale Alpha=&FloatFromCurve(FloatCurve=DataInterface() {propertyValues {Curve {keys [
         {interpMode=RCIM_Cubic, tangentMode=RCTM_User, time=0, value=0}
         <at 0.3 0.5>
         {arriveTangent=-0.0207, interpMode=RCIM_Cubic, leaveTangent=-0.0207, tangentMode=RCTM_User, time=0.7, value=0.5}
         <at 1 0>
        ]}}})
    Scale RGB=&Multiply_VectorByFloat(Vector=<1 1 1>, Float=1)
    ScaleA=true
    ScaleRGB=true
   )
  ]
  rendererClasses [&NiagaraSpriteRendererProperties]
  renderers [{rendererClass=&NiagaraSpriteRendererProperties}]
 }
 glitchRibbon01 {
  emitterUpdateScript [
   &EmitterState()
   &FN_EmitterDivingAttributes_v1(Accumulate Dive=-0.0025, Accumulate Rest=0, SkeletalMesh:NiagaraDataInterfaceSkeletalMesh=$User.SkeletalMesh)
   &FN_SpawnRibbon_v2(Number of Ribbons=4, Spawn Mode=::NewEnumerator0, SpawnRate=50)
  ]
  particleSpawnScript [
   &FN_SetSpawnGroup_v2(Number of Spawn Groups=4)
   &FN_InitializeRibbon_v3(
    Color:LinearColor=&MultiplyLinearColor(Linear Color=<rgba 0.7991 1 0.1413 1>, Scale Factor=<1 1 1 1>)
    Custom Offset=true
    Enable Skeletal Position=true
    Offset Amount=1
    Ribbon Width=17
   )
   /SetVariables(Particles.RibbonNumber:NiagaraInt32=$Particles.SpawnGroup) {isSetParameters=true}
   /SetVariables(Particles.Color=&LerpLinearColors(
     StartColor=&ColorFromCurve(
      ColorCurve=DataInterface() {propertyValues {
        AlphaCurve {keys [<at 0 1 RCIM_Cubic>]}
        BlueCurve {keys [<at 0 1 RCIM_Cubic>,<at 0.7704 0.651>,<at 1 1 RCIM_Cubic>]}
        GreenCurve {keys [<at 0 0.7687 RCIM_Cubic>,<at 0.7704 0.1163>,<at 1 0.0833 RCIM_Cubic>]}
        RedCurve {keys [<at 0 0.0525 RCIM_Cubic>,<at 0.7704 0>,<at 1 0.5317 RCIM_Cubic>]}
       }}
      CurveIndex=&RandomRangeFloat(Minimum=0, Maximum=1)
     )
     EndColor=&ColorFromCurve(
      ColorCurve=DataInterface() {propertyValues {
        AlphaCurve {keys [<at 0 1 RCIM_Cubic>]}
        BlueCurve {keys [<at 0 1 RCIM_Cubic>,<at 0.8279 0.651>,<at 1 1 RCIM_Cubic>]}
        GreenCurve {keys [<at 0 0.7687 RCIM_Cubic>,<at 0.8279 0.1163>,<at 1 0.0833 RCIM_Cubic>]}
        RedCurve {keys [<at 0 0.0525 RCIM_Cubic>,<at 0.8279 0>,<at 1 0.5317 RCIM_Cubic>]}
       }}
      CurveIndex=&RandomRangeFloat(Minimum=0, Maximum=1)
     )
     LerpFactor=&MakeFloatFromInt(Int=$Particles.SpawnGroup)
    )) {isSetParameters=true}
   &Color(Color=$Particles.Color, Scale Color=&RandomRangeVector(Minimum=<2 1 1>, Maximum=<10 1 1>))
   &AddVelocity(Velocity=&Lerp_Vector(A=<0 0 400>, B=<0 0 400>, Alpha=$Emitter.Dive))
   /SetVariables(Particles.SmokeColor:LinearColor=<rgba 0 0 0 0.6>) {isSetParameters=true}
   &SampleSkeletalMeshSkeleton(Mesh:NiagaraDataInterfaceSkeletalMesh=$User.SkeletalMesh)
  ]
  particleUpdateScript [
   &ScaleVelocity(Velocity Scale=&MakeVector(
     X=&FloatFromCurve(FloatCurve=DataInterface() {propertyValues {Curve {keys [<at 0 1 RCIM_Cubic>,<at 0.3 0 RCIM_Cubic>]}}})
     Y=&FloatFromCurve(FloatCurve=DataInterface() {propertyValues {Curve {keys [<at 0 1 RCIM_Cubic>,<at 0.3 0 RCIM_Cubic>]}}})
     Z=1
    ))
   &DynamicMaterialParameters(
    Index 0 Param 1=&Add_Float(A=$Particles.MaterialRandom, B=$Emitter.IntegratedDive)
    Index 0 Param 2=&Add_Float(A=$Particles.MaterialRandom, B=$Emitter.IntegratedDive)
    Index 0 Param 3=&Lerp_Float(A=1, B=0.2, Alpha=$Emitter.Dive)
    Module.Param0WriteEnabled=true
    Module.Param1WriteEnabled=true
    Module.Param2WriteEnabled=true
    Module.Param3WriteEnabled=true
   ) {module=DynamicMaterialParameters001}
   &SolveForcesAndVelocity(Clamp Velocity=false, Limit Acceleration=false)
   &FN_UpdateRibbon(
    Enable Skeletal Position=true
    Initial Color:LinearColor=$Particles.Initial.Color
    Initial RibbonWidth:NiagaraFloat=$Particles.Initial.RibbonWidth
   )
   &ScaleColor(
    Scale Alpha=&FloatFromCurve(FloatCurve=DataInterface() {propertyValues {Curve {keys [<at 0 1>, {arriveTangent=-0.1164, interpMode=RCIM_Cubic, leaveTangent=-0.1164, tangentMode=RCTM_User, time=0.4, value=1}, <at 1 0>]}}})
    Scale RGB=&VectorFromCurve(VectorCurve=DataInterface() {propertyValues {
       XCurve {keys [<at 0.0618 1>,<at 0.1318 1>,<at 0.25 1>,<at 0.8 0.9>]}
       YCurve {keys [<at 0.25 1>,<at 0.75 1>]}
       ZCurve {keys [<at 0.0618 1>,<at 0.1318 1>,<at 0.25 1>,<at 0.8 0.9>]}
      }})
    ScaleA=true
    ScaleRGB=true
   )
   &ScaleColor(
    Scale RGB=&VectorFromFloat(Value=&FloatFromCurve(FloatCurve=DataInterface() {propertyValues {Curve {keys [
          {arriveTangent=0.1804, interpMode=RCIM_Cubic, leaveTangent=0.1804, tangentMode=RCTM_User, time=0, value=6}
          {arriveTangent=-39.5437, interpMode=RCIM_Cubic, leaveTangent=-39.5437, tangentMode=RCTM_User, time=0.3, value=1}
         ]}}}))
    ScaleA=true
    ScaleRGB=true
   ) {module=ScaleColor001}
   &ScaleRibbonWidth(Ribbon Width Scale=&FloatFromCurve(FloatCurve=DataInterface() {propertyValues {Curve {keys [
         <at 0.011 1>
         {arriveTangent=0.0761, interpMode=RCIM_Cubic, leaveTangent=0.0761, tangentMode=RCTM_User, time=0.5, value=1}
         {arriveTangent=-0.5236, interpMode=RCIM_Cubic, leaveTangent=-0.5236, tangentMode=RCTM_User, time=1, value=0.7}
        ]}}}))
  ]
  rendererClasses [&NiagaraRibbonRendererProperties]
  renderers [{rendererClass=&NiagaraRibbonRendererProperties}]
 }
 glitchRibbon02 {
  emitterUpdateScript [
   &EmitterState()
   &FN_SpawnRibbon_v2(Number of Ribbons=2, Spawn Mode=::NewEnumerator0)
  ]
  particleSpawnScript [
   &FN_InitializeRibbon_v3(Enable Skeletal Position=true, Lifetime=0.7, Ribbon Width=14)
   &AddVelocity(Velocity=&FN_FrontEndVector3Modifier(Gameplay=<0 0 410>, Front End=<0 0 350>)) {module=AddVelocity001}
   /SetVariables(Particles.Color=&ColorFromCurve(
     ColorCurve=DataInterface() {propertyValues {
       AlphaCurve {keys [<at 0 1 RCIM_Cubic>]}
       BlueCurve {keys [<at 0 1 RCIM_Cubic>,<at 0.7704 0.651>,<at 1 1 RCIM_Cubic>]}
       GreenCurve {keys [<at 0 0.7687 RCIM_Cubic>,<at 0.7704 0.1163>,<at 1 0.0833 RCIM_Cubic>]}
       RedCurve {keys [<at 0 0.0525 RCIM_Cubic>,<at 0.7704 0>,<at 1 0.5317 RCIM_Cubic>]}
      }}
     CurveIndex=&RandomRangeFloat(Minimum=0, Maximum=1)
    )) {isSetParameters=true}
   &Color(Color=$Particles.Color, Scale Color=&RandomRangeVector(Minimum=<2 1 1>, Maximum=<50 5 1>)) {module=Color001}
   &DynamicMaterialParameters(
    Index 0 Param 1=&RandomRangeFloat(Minimum=0, Maximum=1)
    Module.Param0WriteEnabled=true
    Module.Param1WriteEnabled=true
    Module.Param2WriteEnabled=true
    Module.Param3WriteEnabled=true
   )
  ]
  particleUpdateScript [
   &Drag(Drag=0.9, Use Linear Drag=true, Use Rotational Drag=false)
   /Game/Effects/Niagara/Modules/Skydiving/FN_DivingAttributes/FN_DivingAttributes_v2(SkeletalMesh:NiagaraDataInterfaceSkeletalMesh=$User.SkeletalMesh)
   &SolveForcesAndVelocity(Clamp Velocity=false, Limit Acceleration=false)
   &FN_UpdateRibbon(
    Enable Skeletal Position=true
    Initial Color:LinearColor=$Particles.Initial.Color
    Initial RibbonWidth:NiagaraFloat=$Particles.Initial.RibbonWidth
   )
   &ScaleColor(
    Scale Alpha=&FloatFromCurve(
     FloatCurve=DataInterface() {propertyValues {Curve {keys [<at 0 1>, {arriveTangent=-2.8786, interpMode=RCIM_Cubic, leaveTangent=-2.8786, tangentMode=RCTM_User, time=0.1, value=1}, <at 1 0>]}}}
     Scale Curve=0.15
    )
    Scale RGB=&VectorFromCurve(VectorCurve=DataInterface() {propertyValues {
       XCurve {keys [<at 0.0618 1>,<at 0.1319 1>,<at 0.25 1>,<at 0.8 0.9>]}
       YCurve {keys [<at 0.25 1>,<at 0.75 1>]}
       ZCurve {keys [<at 0.0618 1>,<at 0.1319 1>,<at 0.25 1>,<at 0.8 0.9>]}
      }})
    ScaleA=true
    ScaleRGB=true
   )
   &DynamicMaterialParameters(
    Index 1 Param 1=&Add_Float(
     A=&Multiply_Float(A=1, B=&MakeFloatFromInt(Int=HlslExpression() {hlslExpression=Particles.RibbonID}))
     B=&Multiply_Float(A=-1, B=$Engine.Time)
    )
    Index 1 Param 3=1
    Index 1 Param 4=1
    Index 2 Param 1=0.5
    Index 2 Param 2=0.5
    Index 3 Param 1=0.1
    Index 3 Param 2=1
    Index 3 Param 3=1
    Index 3 Param 4=1
    Module.Param0WriteEnabled1=true
    Module.Param0WriteEnabled2=true
    Module.Param0WriteEnabled3=true
    Module.Param1WriteEnabled1=true
    Module.Param1WriteEnabled2=true
    Module.Param1WriteEnabled3=true
    Module.Param2WriteEnabled1=true
    Module.Param2WriteEnabled2=true
    Module.Param2WriteEnabled3=true
    Module.Param3WriteEnabled1=true
    Module.Param3WriteEnabled2=true
    Module.Param3WriteEnabled3=true
    Write Parameter Index 0=false
    Write Parameter Index 1=true
    Write Parameter Index 2=true
    Write Parameter Index 3=true
   ) {module=DynamicMaterialParameters001}
  ]
  rendererClasses [&NiagaraRibbonRendererProperties]
  renderers [{rendererClass=&NiagaraRibbonRendererProperties}]
 }
 lights_hands {
  emitterSpawnScript [&FN_PreserveSkydiveUserVariables() {module=PreserveUserVariables}]
  emitterUpdateScript [
   &EmitterState(
    Enable Distance Culling=true
    Life Cycle Mode=::NewEnumerator1
    Loop Behavior=::NewEnumerator1
    Loop Duration Mode=::NewEnumerator1
    Module.bUseMaxDistance=true
    Module.bUseMinDistance=false
    Reset Age On Awaken=true
    Scalability Mode=::NewEnumerator1
    UseLoopDelay=false
   )
   /Game/Effects/Niagara/Modules/Spawn/SpawnRateFromMultipleSpawnGroups_2(Global Spawnrate=2)
   &SpawnRateFromMultipleSpawnGroups_4() {enabled=false}
  ]
  particleSpawnScript [
   &FN_DivingAttributes_v2(Bone Index=2, SkeletalMesh:NiagaraDataInterfaceSkeletalMesh=$User.SkeletalMesh)
   &FN_SetSpawnGroup_v2(Number of Spawn Groups=2)
   &InitializeParticle(
    Color=&ColorFromCurve(
     ColorCurve=DataInterface() {propertyValues {
       AlphaCurve {keys [<at 0 1 RCIM_Cubic>,<at 1 1 RCIM_Cubic>]}
       BlueCurve {keys [
         <at 0 1>
         <at 0.2363 0.651>
         {arriveTangent=-1.1619, interpMode=RCIM_Cubic, leaveTangent=-1.1619, time=0.5721, value=0.651}
         <at 0.8607 0>
         <at 1 1 RCIM_Cubic>
        ]}
       GreenCurve {keys [
         <at 0 0.7687>
         <at 0.2363 0.1163>
         {arriveTangent=-0.1972, interpMode=RCIM_Cubic, leaveTangent=-0.1972, time=0.5721, value=0.1163}
         <at 0.8607 0.599>
         <at 1 0.0833 RCIM_Cubic>
        ]}
       RedCurve {keys [
         <at 0 0.0525>
         <at 0.2363 0>
         {arriveTangent=0.2803, interpMode=RCIM_Cubic, leaveTangent=0.2803, time=0.5721, value=0}
         <at 0.8607 0.2938>
         <at 1 0.5317 RCIM_Cubic>
        ]}
      }}
     CurveIndex=&RandomRangeFloat(Minimum=0, Maximum=1)
    )
    Lifetime=&RandomRangeFloat(Minimum=0.35, Maximum=0.5)
    Mesh Scale=&RandomRangeVector(Minimum=<0.04 0.04 0.05>, Maximum=<0.065 0.065 0.06>)
    Module.Write Color=true
    Module.Write Lifetime=true
    Module.Write Mass=true
    Module.Write Material Random=true
    Module.Write Position=false
    Module.Write Scale=true
    Module.Write SpriteRotation=false
    Module.Write SpriteSize=false
    Position=&SimulationPosition()
   )
   /Niagara/Modules/Spawn/Orientation/V2/InitialMeshOrientation(
    Orientation Axis:Vector3f=<0 0 1>
    Orientation Coordinate Space:ENiagaraCoordinateSpace=::World
    Orientation Vector=<0 0 1>
    Rotation=&MakeVector(X=0, Y=0, Z=&UniformRangedFloat(Minimum=0, Maximum=1))
    Rotation Coordinate Space:ENiagaraMeshTransforms=::NewEnumerator1
    Use Orientation Vector=true
    Use Rotation Vector=true
   )
   /SetVariables(Particles.PositionOffsetArray=DataInterface() {propertyValues {GpuSyncMode=SyncCpuToGpu, MaxElements=0, FloatData [<-20 0 0>, <20 0 0>]}}) {isSetParameters=true, module=SetVariables_#1}
  ]
  particleUpdateScript [
   &ParticleState()
   &FN_DivingAttributes_v2(SkeletalMesh:NiagaraDataInterfaceSkeletalMesh=$User.SkeletalMesh) {module=FN_DivingAttributes_v2001}
   &SampleSkeletalMeshSkeleton(
    Mesh:NiagaraDataInterfaceSkeletalMesh=$User.SkeletalMesh
    Sampled Rotation Coordinate Space:ENiagaraCoordinateSpace=::World
    Specific Bone Index:NiagaraInt32=$Particles.SpawnGroup
   )
   &SkeletalMeshLocation(
    Apply Sampled Orientation to Mesh Particles=true
    Bone / Socket Index=$Particles.SampleSkeletalMeshSkeleton.SampledBoneIndex
    Bone Sampling Mode=::NewEnumerator6
    Inherit Velocity=false
    MaskSpawnGroup=false
    Sampled Position Offset=&self:PositionOffset(ArrayIndex:NiagaraInt32=$Particles.SpawnGroup, OffsetArray:NiagaraDataInterfaceArrayFloat3=$Particles.PositionOffsetArray)
    Sampled Position Offset Coordinate Space:ENiagaraMeshTransforms=::NewEnumerator3
    Skeletal Mesh=$User.SkeletalMesh
    Write Sampled Mesh Attrributes to Particle Attributes=true
   )
   /Niagara/Modules/Update/Orientation/V2/OrientMeshToVector(
    Override Incoming Orientation=1
    Override Quat=true
    Reference Axis Vector:Vector3f=<0 0 1>
    UseOptionalSideAxis=false
    UseOptionalUpAxis=false
   )
   &ScaleMeshSize(Scale Factor=&Lerp_Vector(A=<1 1 1>, B=<0.5 0.5 5>, Alpha=$Particles.Dive))
   &ScaleColor(
    Scale Alpha=&FloatFromCurve(FloatCurve=DataInterface() {propertyValues {Curve {keys [<at 0 1>,<at 0.3 1>,<at 1 1>]}}})
    ScaleA=true
    ScaleRGB=true
   )
   &DynamicMaterialParameters(
    Index 0 Param 1=&FloatFromCurve(FloatCurve=DataInterface() {propertyValues {Curve {keys [<at 0 1>,<at 0.3 1>,<at 0.6 1>,<at 1 0>]}}})
    Index 0 Param 2=1
    Index 0 Param 3=2
    Module.Param0WriteEnabled=true
    Module.Param1WriteEnabled=true
    Module.Param2WriteEnabled=true
    Module.Param3WriteEnabled=true
   )
   &FN_Light_v3(
    Color=&ColorFromCurve(
     ColorCurve=DataInterface() {propertyValues {
       AlphaCurve {keys [<at 0 1 RCIM_Cubic>,<at 1 1 RCIM_Cubic>]}
       BlueCurve {keys [<at 0 0.025 RCIM_Cubic>,<at 0.1792 1>,<at 0.3909 0>,<at 0.8099 0.651>,<at 1 1 RCIM_Cubic>]}
       GreenCurve {keys [<at 0 0.8904 RCIM_Cubic>,<at 0.1792 0.7687>,<at 0.3909 0.599>,<at 0.8099 0.1163>,<at 1 0.0833 RCIM_Cubic>]}
       RedCurve {keys [<at 0 1 RCIM_Cubic>,<at 0.1792 0.0525>,<at 0.3909 0.2938>,<at 0.8099 0>,<at 1 0.5317 RCIM_Cubic>]}
      }}
     CurveIndex=&RandomRangeFloat(Minimum=0, Maximum=1)
    )
    Create Color Attribute=true
    Create Exponent Attribute=true
    Create Radius Attribute=true
    Enable Particle Light Enable=true
    Radius=75
    Scale Color:NiagaraFloat=&FloatFromCurve(FloatCurve=DataInterface() {propertyValues {Curve {keys [
         {interpMode=RCIM_Cubic, tangentMode=RCTM_User, time=0, value=0}
         <at 0.3 1>
         {arriveTangent=-0.007, interpMode=RCIM_Cubic, leaveTangent=-0.007, tangentMode=RCTM_User, time=0.7, value=1}
         {arriveTangent=-0.0564, interpMode=RCIM_Cubic, leaveTangent=-0.0564, tangentMode=RCTM_User, time=1, value=0}
        ]}}})
   )
  ]
  rendererClasses [&NiagaraLightRendererProperties]
  renderers [{rendererClass=&NiagaraLightRendererProperties, rendererIndex=0}]
 }
 meshShapes {
  emitterSpawnScript [&FN_PreserveSkydiveUserVariables() {module=PreserveUserVariables}]
  emitterUpdateScript [
   &EmitterState(
    Enable Distance Culling=true
    Life Cycle Mode=::NewEnumerator1
    Loop Behavior=::NewEnumerator1
    Loop Duration Mode=::NewEnumerator1
    Module.bUseMaxDistance=true
    Module.bUseMinDistance=false
    Reset Age On Awaken=true
    Scalability Mode=::NewEnumerator1
    UseLoopDelay=false
   )
   &SpawnRateFromMultipleSpawnGroups_4(Global Spawnrate=15)
  ]
  particleSpawnScript [
   &FN_DivingAttributes_v2(Bone Index=2, SkeletalMesh:NiagaraDataInterfaceSkeletalMesh=$User.SkeletalMesh)
   &FN_SetSpawnGroup_v2(Number of Spawn Groups=4)
   /Niagara/Modules/Spawn/Initialization/V2/InitializeParticle(
    Color=&ColorFromCurve(
     ColorCurve=DataInterface() {propertyValues {
       AlphaCurve {keys [<at 0 1 RCIM_Cubic>,<at 1 1 RCIM_Cubic>]}
       BlueCurve {keys [
         <at 0 1>
         <at 0.2363 0.651>
         {arriveTangent=-1.0402, interpMode=RCIM_Cubic, leaveTangent=-1.0402, time=0.5915, value=0.651}
         <at 0.8621 0>
         <at 0.9549 1 RCIM_Cubic>
        ]}
       GreenCurve {keys [
         <at 0 0.7687>
         <at 0.2363 0.1163>
         {arriveTangent=0.7711, interpMode=RCIM_Cubic, leaveTangent=0.7711, time=0.5915, value=0.1163}
         <at 0.8621 0.599>
         <at 0.9549 0.0833 RCIM_Cubic>
        ]}
       RedCurve {keys [
         <at 0 0.0525>
         <at 0.2363 0>
         {arriveTangent=0.4695, interpMode=RCIM_Cubic, leaveTangent=0.4695, time=0.5915, value=0}
         <at 0.8621 0.2938>
         <at 0.9549 0.5317 RCIM_Cubic>
        ]}
      }}
     CurveIndex=&RandomRangeFloat(Minimum=0, Maximum=1)
    )
    Lifetime Max=0.6
    Lifetime Min=0.35
    Lifetime Mode=::NewEnumerator1
    Mesh Scale=&FN_FrontEndVector3Modifier(
     Gameplay=&RandomRangeVector(Minimum=<0.04 0.04 0.05>, Maximum=<0.065 0.065 0.06>)
     Front End=&RandomRangeVector(Minimum=<0.03 0.03 0.04>, Maximum=<0.05 0.05 0.05>)
    )
    Mesh Scale Mode=::NewEnumerator4
    UsePositionOffset=false
    Write Material Random=true
   )
   /Niagara/Modules/Spawn/Initialization/InitializeParticle(
    Color=&ColorFromCurve(
     ColorCurve=DataInterface() {propertyValues {
       AlphaCurve {keys [<at 0 1 RCIM_Cubic>,<at 1 1 RCIM_Cubic>]}
       BlueCurve {keys [
         <at 0 1>
         <at 0.2363 0.651>
         {arriveTangent=-1.1619, interpMode=RCIM_Cubic, leaveTangent=-1.1619, time=0.5721, value=0.651}
         <at 0.8607 0>
         <at 1 1 RCIM_Cubic>
        ]}
       GreenCurve {keys [
         <at 0 0.7687>
         <at 0.2363 0.1163>
         {arriveTangent=-0.1972, interpMode=RCIM_Cubic, leaveTangent=-0.1972, time=0.5721, value=0.1163}
         <at 0.8607 0.599>
         <at 1 0.0833 RCIM_Cubic>
        ]}
       RedCurve {keys [
         <at 0 0.0525>
         <at 0.2363 0>
         {arriveTangent=0.2803, interpMode=RCIM_Cubic, leaveTangent=0.2803, time=0.5721, value=0}
         <at 0.8607 0.2938>
         <at 1 0.5317 RCIM_Cubic>
        ]}
      }}
     CurveIndex=&RandomRangeFloat(Minimum=0, Maximum=1)
    )
    Lifetime=&RandomRangeFloat(Minimum=0.35, Maximum=0.6)
    Mesh Scale=&FN_FrontEndVector3Modifier(
     Gameplay=&RandomRangeVector(Minimum=<0.04 0.04 0.05>, Maximum=<0.065 0.065 0.06>)
     Front End=&RandomRangeVector(Minimum=<0.03 0.03 0.04>, Maximum=<0.05 0.05 0.05>)
    )
    Module.Write Color=true
    Module.Write Lifetime=true
    Module.Write Mass=true
    Module.Write Material Random=true
    Module.Write Position=false
    Module.Write Scale=true
    Module.Write SpriteRotation=false
    Module.Write SpriteSize=false
    Position=&SimulationPosition()
   ) {enabled=false, module=InitializeParticle001}
   &InitialMeshOrientation(
    Orientation Axis:Vector3f=<0 0 1>
    Orientation Coordinate Space:ENiagaraCoordinateSpace=::World
    Orientation Vector=<0 0 1>
    Rotation=&MakeVector(X=0, Y=0, Z=&UniformRangedFloat(Minimum=0, Maximum=1))
    Rotation Coordinate Space:ENiagaraMeshTransforms=::NewEnumerator1
    Use Orientation Vector=true
    Use Rotation Vector=true
   )
   /SetVariables(Particles.PositionOffsetArray=DataInterface() {propertyValues {GpuSyncMode=SyncCpuToGpu, MaxElements=0, FloatData [<-20 0 0>, <20 0 0>, <20 0 0>, <-20 0 0>]}}) {isSetParameters=true, module=SetVariables_#1}
  ]
  particleUpdateScript [
   &ParticleState()
   &FN_DivingAttributes_v2(SkeletalMesh:NiagaraDataInterfaceSkeletalMesh=$User.SkeletalMesh) {module=FN_DivingAttributes_v2001}
   &SampleSkeletalMeshSkeleton(
    Mesh:NiagaraDataInterfaceSkeletalMesh=$User.SkeletalMesh
    Sampled Rotation Coordinate Space:ENiagaraCoordinateSpace=::World
    Specific Bone Index:NiagaraInt32=$Particles.SpawnGroup
   )
   &SkeletalMeshLocation(
    Apply Sampled Orientation to Mesh Particles=true
    Bone / Socket Index=$Particles.SampleSkeletalMeshSkeleton.SampledBoneIndex
    Bone Sampling Mode=::NewEnumerator6
    Inherit Velocity=false
    MaskSpawnGroup=false
    Sampled Position Offset=&self:PositionOffset(ArrayIndex:NiagaraInt32=$Particles.SpawnGroup, OffsetArray:NiagaraDataInterfaceArrayFloat3=$Particles.PositionOffsetArray)
    Sampled Position Offset Coordinate Space:ENiagaraMeshTransforms=::NewEnumerator3
    Skeletal Mesh=$User.SkeletalMesh
    Write Sampled Mesh Attrributes to Particle Attributes=true
   )
   &OrientMeshToVector(
    Override Incoming Orientation=1
    Override Quat=true
    Reference Axis Vector:Vector3f=<0 0 1>
    UseOptionalSideAxis=false
    UseOptionalUpAxis=false
   )
   &ScaleMeshSize(Scale Factor=&Lerp_Vector(A=<1 1 1>, B=<0.5 0.5 5>, Alpha=$Particles.Dive))
   &DynamicMaterialParameters(
    Index 0 Param 1=&FloatFromCurve(FloatCurve=DataInterface() {propertyValues {Curve {keys [<at 0 1>,<at 0.3 1>,<at 0.6 1>,<at 1 0>]}}})
    Index 0 Param 2=1
    Index 0 Param 3=2
    Module.Param0WriteEnabled=true
    Module.Param1WriteEnabled=true
    Module.Param2WriteEnabled=true
    Module.Param3WriteEnabled=true
   )
  ]
  rendererClasses [&NiagaraMeshRendererProperties]
  renderers [{rendererClass=&NiagaraMeshRendererProperties, rendererIndex=0}]
 }
 ribbon_thin {
  emitterUpdateScript [
   &EmitterState()
   &FN_SpawnRibbon_v2(Number of Ribbons=2, Spawn Mode=::NewEnumerator0)
  ]
  particleSpawnScript [
   &FN_InitializeRibbon_v3(Enable Skeletal Position=true, Lifetime=0.7, Ribbon Width=12)
   &AddVelocity(Velocity=&FN_FrontEndVector3Modifier(Gameplay=<0 0 410>, Front End=<0 0 350>)) {module=AddVelocity001}
   &DynamicMaterialParameters(
    Index 0 Param 1=&RandomRangeFloat(Minimum=0, Maximum=1)
    Module.Param0WriteEnabled=true
    Module.Param1WriteEnabled=true
    Module.Param2WriteEnabled=true
    Module.Param3WriteEnabled=true
   )
   &Color(Color=&ColorFromCurve(
     ColorCurve=DataInterface() {propertyValues {
       AlphaCurve {keys [<at 0 1 RCIM_Cubic>,<at 0.3655 1>,<at 1 1 RCIM_Cubic>]}
       BlueCurve {keys [<at 0.4031 0.651 RCIM_Cubic>,<at 0.748 1>,<at 1 1 RCIM_Cubic>]}
       GreenCurve {keys [<at 0.4031 0.1163 RCIM_Cubic>,<at 0.748 0.7687>,<at 1 0.0833 RCIM_Cubic>]}
       RedCurve {keys [<at 0.4031 0 RCIM_Cubic>,<at 0.748 0.0525>,<at 1 0.5317 RCIM_Cubic>]}
      }}
     CurveIndex=&RandomRangeFloat(Minimum=0, Maximum=1)
    )) {module=Color001}
  ]
  particleUpdateScript [
   &Drag(Drag=0.9, Use Linear Drag=true, Use Rotational Drag=false)
   &FN_DivingAttributes_v2(SkeletalMesh:NiagaraDataInterfaceSkeletalMesh=$User.SkeletalMesh)
   &SolveForcesAndVelocity(Clamp Velocity=false, Limit Acceleration=false)
   &FN_UpdateRibbon(
    Enable Skeletal Position=true
    Initial Color:LinearColor=$Particles.Initial.Color
    Initial RibbonWidth:NiagaraFloat=$Particles.Initial.RibbonWidth
   )
   &DynamicMaterialParameters(
    Index 1 Param 1=&Add_Float(
     A=&Multiply_Float(A=1, B=&MakeFloatFromInt(Int=HlslExpression() {hlslExpression=Particles.RibbonID}))
     B=&Multiply_Float(A=-1, B=$Engine.Time)
    )
    Index 1 Param 3=1
    Index 1 Param 4=1
    Index 2 Param 1=0.5
    Index 2 Param 2=0.5
    Index 3 Param 1=0.1
    Index 3 Param 2=1
    Index 3 Param 3=1
    Index 3 Param 4=1
    Module.Param0WriteEnabled1=true
    Module.Param0WriteEnabled2=true
    Module.Param0WriteEnabled3=true
    Module.Param1WriteEnabled1=true
    Module.Param1WriteEnabled2=true
    Module.Param1WriteEnabled3=true
    Module.Param2WriteEnabled1=true
    Module.Param2WriteEnabled2=true
    Module.Param2WriteEnabled3=true
    Module.Param3WriteEnabled1=true
    Module.Param3WriteEnabled2=true
    Module.Param3WriteEnabled3=true
    Write Parameter Index 0=false
    Write Parameter Index 1=true
    Write Parameter Index 2=true
    Write Parameter Index 3=true
   ) {module=DynamicMaterialParameters001}
   &Color(
    Color=&ColorFromCurve(ColorCurve=DataInterface() {propertyValues {
       AlphaCurve {keys [<at 0.3766 1 RCIM_Cubic>,<at 0.9857 -0.003 RCIM_Cubic>]}
       BlueCurve {keys [<at -0.0023 1 RCIM_Cubic>,<at 0.3035 1>]}
       GreenCurve {keys [<at -0.0023 0.7687 RCIM_Cubic>,<at 0.3035 0.1099>]}
       RedCurve {keys [<at -0.0023 0.0525 RCIM_Cubic>,<at 0.3035 0.078>]}
      }})
    Scale Alpha=&FloatFromCurve(
     FloatCurve=DataInterface() {propertyValues {Curve {keys [
         {arriveTangent=0.0265, interpMode=RCIM_Cubic, leaveTangent=0.0265, tangentMode=RCTM_User, time=0, value=1}
         {arriveTangent=-0.3873, interpMode=RCIM_Cubic, leaveTangent=-0.3872, tangentMode=RCTM_User, time=1, value=0}
        ]}}}
     Scale Curve=0.5
    )
    Scale Color=<2 1 1>
   ) {module=Color002}
  ]
  rendererClasses [&NiagaraRibbonRendererProperties]
  renderers [{rendererClass=&NiagaraRibbonRendererProperties}]
 }
 speedlines_dive {
  simTarget=GPUComputeSim
  emitterSpawnScript [&FN_PreserveSkydiveUserVariables() {module=PreserveUserVariables}]
  emitterUpdateScript [
   &EmitterState(Life Cycle Mode=::NewEnumerator1, Loop Duration=1, Scalability Mode=::NewEnumerator1, UseLoopDelay=false)
   &FN_EmitterDivingAttributes_v1(Accumulate Dive=-0.0025, Accumulate Rest=0, SkeletalMesh:NiagaraDataInterfaceSkeletalMesh=$User.SkeletalMesh)
   &SpawnRate(SpawnRate=&Multiply_Float(A=50, B=$Emitter.Dive), Use Spawn Probability=false) {module=SpawnRate001}
  ]
  particleSpawnScript [
   &InitializeParticle(
    Color=&ColorFromCurve(
     ColorCurve=DataInterface() {propertyValues {
       AlphaCurve {keys [<at 0 1 RCIM_Cubic>,<at 1 1 RCIM_Cubic>]}
       BlueCurve {keys [<at 0 0>, <at 0.225 0.651>, {arriveTangent=0.4503, interpMode=RCIM_Cubic, leaveTangent=0.4503, time=0.6172, value=1}, <at 1 1 RCIM_Cubic>]}
       GreenCurve {keys [
         <at 0 0.599>
         <at 0.225 0.1163>
         {arriveTangent=-0.0426, interpMode=RCIM_Cubic, leaveTangent=-0.0426, time=0.6172, value=0.1099}
         <at 1 0.0833 RCIM_Cubic>
        ]}
       RedCurve {keys [
         <at 0 0.2938>
         <at 0.225 0>
         {arriveTangent=0.6861, interpMode=RCIM_Cubic, leaveTangent=0.6861, time=0.6172, value=0.078}
         <at 1 0.5317 RCIM_Cubic>
        ]}
      }}
     CurveIndex=&RandomRangeFloat(Minimum=0, Maximum=1)
    )
    Lifetime=&UniformRangedFloat(Minimum=0.2, Maximum=0.2)
    Mesh Scale=&Multiply_Vector(
     A=<1 1 1>
     B=&VectorFromFloat(Value=&UniformRangedFloat(Minimum=0.2, Maximum=1))
    )
    Module.Write Color=true
    Module.Write Lifetime=true
    Module.Write Mass=true
    Module.Write Material Random=true
    Module.Write Position=false
    Module.Write Scale=false
    Module.Write SpriteRotation=true
    Module.Write SpriteSize=true
    Position=&SimulationPosition()
    Sprite Rotation=90
    Sprite Size=&MakeVector2D(
     X=&UniformRangedFloat(Minimum=0.75, Maximum=2.5)
     Y=&UniformRangedFloat(Minimum=250, Maximum=400)
    )
   )
   &CylinderLocation(
    Cylinder Height=5
    Cylinder Radius=40
    MaskSpawnGroup=false
    Module.Enable Lathe Profile=false
    Module.Override Randomness=false
    Module.Surface Only=true
    Surface Only Band Thickness=1
   )
   &AddVelocity(
    Scale Added Velocity=&Lerp_Vector(
     A=<0 0 1>
     B=<0 0 3>
     Alpha=&FloatFromCurve(
      FloatCurve=DataInterface() {propertyValues {Curve {keys [<at 0 0>,<at 1 1>]}}}
      CurveIndex=&Divide_Float(A=&VectorLength(Vector=$Particles.FallingSpeed), B=3000)
     )
    )
    Velocity=<0 0 1000>
   )
   /SetVariables(Particles.RandAlpha:NiagaraFloat=&RandomRangeFloat(Minimum=0.3, Maximum=1)) {isSetParameters=true}
  ]
  particleUpdateScript [
   &ParticleState()
   &SolveForcesAndVelocity(Clamp Velocity=false, Limit Acceleration=false)
   &ScaleColor(
    Scale Alpha=&FloatFromCurve(
     FloatCurve=DataInterface() {propertyValues {Curve {keys [<at 0 0 RCIM_Cubic>,<at 0.3634 1.0074 RCIM_Cubic>,<at 1 0 RCIM_Cubic>]}}}
     Scale Curve=$Particles.RandAlpha
    )
    ScaleA=true
    ScaleRGB=true
   )
  ]
  rendererClasses [&NiagaraSpriteRendererProperties]
  renderers [{rendererClass=&NiagaraSpriteRendererProperties}]
 }
}
summary {
 systemName=NS_Contrail_ID106_GeoStorm
 emitters [
  {emitterName=glitchRibbon01, rendererClasses [&NiagaraRibbonRendererProperties]}
  {emitterName=glitchRibbon02, rendererClasses [&NiagaraRibbonRendererProperties]}
  {emitterName=meshShapes, rendererClasses [&NiagaraMeshRendererProperties]}
  {emitterName=decal, rendererClasses [&NiagaraSpriteRendererProperties]}
  {emitterName=ribbon_thin, rendererClasses [&NiagaraRibbonRendererProperties]}
  {emitterName=bits, rendererClasses [&NiagaraSpriteRendererProperties]}
  {emitterName=speedlines_dive, simTarget=GPUComputeSim, rendererClasses [&NiagaraSpriteRendererProperties]}
  {emitterName=lights_hands, rendererClasses [&NiagaraLightRendererProperties]}
 ]
 userVariables [
  User.bIsFrontEnd {type=NiagaraBool}
  User.bIsFrontEndPreview {type=NiagaraBool}
  User.PlayerSpeed {type=NiagaraFloat, defaultValue=0, description=""}
  User.RibbonIDOffset {type=NiagaraInt32}
  User.SkeletalMesh {
   type=NiagaraDataInterfaceSkeletalMesh
   defaultValue {
    struct=NiagaraExt_VariableValue_DataInterface
    value {dataInterface=&self:NiagaraDataInterfaceSkeletalMesh_0}
   }
  }
 ]
}
